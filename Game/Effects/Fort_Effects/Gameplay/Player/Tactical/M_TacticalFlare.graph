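# Translucent, unlit, two-sided Niagara sprite material, named for a tactical flare effect, using a 2x2 impact-flash SubUV texture.
# Particle color drives emissive through HDR-from-LDR remapping and FX emissive adjustments.
# Opacity subtracts tiled, fixed-mip noise, shaped by dynamic Erode, from the flash, then applies particle alpha, depth fade (skipped on Low quality) and near-camera fade.
# The SoftTexture lerp blends identical inputs, so it has no effect.
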Material /Game/Effects/Fort_Effects/Gameplay/Player/Tactical/M_TacticalFlare
graph {
 properties {
  bAutomaticallySetUsageInEditor=false
  bUsedWithNiagaraSprites=true
  blendMode=BLEND_Translucent
  shadingModel=MSM_Unlit
  twoSided=true
  usedShadingModels=Unlit
 }
 nodes [
  n1 = ParticleColor()
  MaterialFunctionCall(Emissive=MaterialFunctionCall(LDR Input=Saturate(None=@n1.RGB), HDR Tint=@n1.RGB) {materialFunction=&DeriveHDRfromLDR}.RemappedRange) {materialFunction=&MF_FX_EmissiveAdjustments}
  n2 = ParticleSubUV!(UVs, Tex, Apply View MipBias) {texture=/Game/Effects/Fort_Effects/Textures/Explo/T_Impact_Flash_2x2_01}
  n3 = Multiply(
   A=TextureCoordinate()
   B=DynamicParameter() {defaultValue=<rgba 1 0.75 1 1>, paramNames [Erode, Noise Tiling, SoftTexture, Param4]}.Noise Tiling
  )
  n4 = ScalarParameter() {defaultValue=2, parameterName=Mip Level}
  n5 = Subtract(
   A=LinearInterpolate(
    A=@n2.G
    B=@n2.G
    Alpha=DynamicParameter() {defaultValue=<rgba 1 0.75 1 1>, paramNames [Erode, Noise Tiling, SoftTexture, Param4]}.SoftTexture
   )
   B=Power(
    Base=TextureSample!(UVs=@n3, Tex, Level=@n4, Apply View MipBias) {mipValueMode=TMVM_MipLevel, samplerType=SAMPLERTYPE_LinearColor, texture=/Game/Global/Textures/Default/Noise/T_Noise01_Inverse}.G
    Exp=DynamicParameter() {defaultValue=<rgba 1 0.75 1 1>, paramNames [Erode, Noise Tiling, SoftTexture, Param4]}.Erode
   )
  )
  n6 = DepthFade(FadeDistance=Constant() {r=12})
  n7 = Constant() {r=1}
  Multiply(
   A=Multiply(A=Multiply(A=@n1.A, B=@n5), B=QualitySwitch(Default=@n6, Low=@n7, High=@n6, Medium=@n6))
   B=MaterialFunctionCall(Distance=ScalarParameter() {defaultValue=150, parameterName=Camera Fade Distance}) {materialFunction=&Near_Camera_Fade}.Result
  )
 ]
}
