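# Niagara system for the tactical elimination variant of the player health and shield damage effect.
# It layers Circle, SharpFlash and Streaks emitters, with a Headshot_Cubes emitter present but disabled.
# A User.bisHeadshot bool is exposed, alongside bool-driven dynamic inputs for int, float and color values that suggest headshot-specific styling.
# Curves, random ranges and an LOD distance remap input are also used to shape values.
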NiagaraSystem /Game/Effects/Fort_Effects/Gameplay/Player/Tactical/NS_Player_HealthShield_Damage_Effect_TacElim
dependencies {
 usedDataInterfaces [&NiagaraDataInterfaceCurve]
 usedDynamicInputs [
  &Lerp_Float
  &FloatFromCurve
  &FN_LOD_Distance_Remap
  &MakeCustomIntFromBool
  &RandomRangeInt
  &MakeCustomLinearColorFromBool
  &Multiply_Vector2D_ByFloat
  &Vector2DFromFloat
  &MakeCustomFloatFromBool
  &RandomRangeFloat
  &RandomRangeLinearColor
 ]
 usedModules [
  &self:SystemSpawnScript.SystemScriptSource.SystemScriptGraph.NiagaraNodeAssignment_1.SetVariables_#1
  &self:SystemSpawnScript.SystemScriptSource.SystemScriptGraph.NiagaraNodeAssignment_0.SetVariables_#2
  &SystemState
 ]
}
emitters {
 Circle {rendererClasses [], renderers []}
 Headshot_Cubes {bEnabled=false, rendererClasses [], renderers []}
 SharpFlash {rendererClasses [], renderers []}
 Streaks {rendererClasses [], renderers []}
}
summary {
 systemName=NS_Player_HealthShield_Damage_Effect_TacElim
 emitters [
  {emitterName=Circle, rendererClasses []}
  {bEnabled=false, emitterName=Headshot_Cubes, rendererClasses []}
  {emitterName=SharpFlash, rendererClasses []}
  {emitterName=Streaks, rendererClasses []}
 ]
 userVariables [User.bisHeadshot {type=NiagaraBool}]
}
